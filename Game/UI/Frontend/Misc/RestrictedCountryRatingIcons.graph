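# Frontend UI widget blueprint for restricted-country rating icons, based on the native FortniteUI RestrictedCountryRatingIcons class.
# Its only widget is the root HorizontalBox HBox_RatingIcons, marked as inherited, with no icons placed in it.
# The event graph and variable list are empty, so any icon population or country logic sits outside this blueprint.
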WidgetBlueprint /Game/UI/Frontend/Misc/RestrictedCountryRatingIcons
graphs [EventGraph {links [], nodes []}]
variables []
widgets {
 info {
  inheritedWidgetCount=1
  parentClass=/Script/FortniteUI.RestrictedCountryRatingIcons
  rootWidgetClass=&HorizontalBox
  widgetCount=1
 }
 widgets [HBox_RatingIcons {widgetClassPath=&HorizontalBox, bInherited=true, widget=HBox_RatingIcons}]
}
